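# Physics asset for the ClumsyChew Law medium male player body cosmetic, set up for secondary motion on its jacket.
# Seven two-box chains at front, sides and back hang from small MSTR anchor spheres through swing- and twist-limited constraints.
# Cross-constraints link neighboring lower segments from front to back, mostly with free rotation, tying the hem together.
# Capsules cover pelvis, lower spine and thighs, plus a pistol holder body that has no constraint.
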PhysicsAsset /BRCosmetics/Characters/Player/Male/Medium/Bodies/M_MED_ClumsyChew_Law/Meshes/M_MED_ClumsyChew_Law_Physics
bodies {
 MSTR_jacket_back [{shapeType=Sphere, radius=0.505, shapeName=MSTR_jacket_back_sphere}]
 MSTR_jacket_back_l [{shapeType=Sphere, radius=0.505, shapeName=MSTR_jacket_back_l_sphere}]
 MSTR_jacket_back_r [{shapeType=Sphere, radius=0.505, shapeName=MSTR_jacket_back_r_sphere}]
 MSTR_jacket_front_l [{shapeType=Sphere, radius=0.505, shapeName=MSTR_jacket_front_l_sphere}]
 MSTR_jacket_front_r [{shapeType=Sphere, radius=0.505, shapeName=MSTR_jacket_front_r_sphere}]
 MSTR_jacket_side_l [{shapeType=Sphere, radius=0.505, shapeName=MSTR_jacket_side_l_sphere}]
 MSTR_jacket_side_r [{shapeType=Sphere, radius=0.505, shapeName=MSTR_jacket_side_r_sphere}]
 dyn_jacket_back_01 [{
   shapeType=Box
   extentX=3.0702
   extentY=4.4432
   extentZ=16.6085
   shapeName=dyn_jacket_back_01_box
   center=<1.3136 0.4407 -1.1625>
   rotation=<rot -7.2038 -90 -0.3371>
  }]
 dyn_jacket_back_01_l [{
   shapeType=Box
   extentX=2.8044
   extentY=6.1403
   extentZ=11.2066
   shapeName=dyn_jacket_side_01_l_box
   center=<2.5103 0.2813 0.2818>
   rotation=<rot 0.2489 -95.2596 4.8406>
  }]
 dyn_jacket_back_01_r [{
   shapeType=Box
   extentX=3.0858
   extentY=6.7628
   extentZ=11.2066
   shapeName=dyn_jacket_side_01_r_box
   center=<2.5666 0.3299 -0.2844>
   rotation=<rot 0.2489 -95.2596 175.9883>
  }]
 dyn_jacket_back_02 [{
   shapeType=Box
   extentX=2.5007
   extentY=4.8726
   extentZ=15.5054
   shapeName=dyn_jacket_back_02_box
   center=<1.7307 0.2861 -1.3378>
   rotation=<rot -7.4594 -90 -0.1361>
  }]
 dyn_jacket_back_02_l [{
   shapeType=Box
   extentX=2.663
   extentY=4.4773
   extentZ=9.656
   shapeName=dyn_jacket_front_02_l_box
   center=<1.98 -0.2316 0.0256>
   rotation=<rot 0.9492 -95.7125 11.2456>
  }]
 dyn_jacket_back_02_r [{
   shapeType=Box
   extentX=2.663
   extentY=4.224
   extentZ=9.656
   shapeName=dyn_jacket_front_02_r_box
   center=<2.2738 0.2092 0.3935>
   rotation=<rot 0.9492 -95.7125 173.0124>
  }]
 dyn_jacket_front_01_l [{
   shapeType=Box
   extentX=1.8258
   extentY=9.1029
   extentZ=10.1501
   shapeName=dyn_jacket_front_01_l_box
   center=<4.7282 -0.117 0.0269>
   rotation=<rot 6.2807 -90.2935 -0.6487>
  }]
 dyn_jacket_front_01_r [{
   shapeType=Box
   extentX=1.8258
   extentY=9.1029
   extentZ=10.1501
   shapeName=dyn_jacket_front_01_l_box
   center=<4.7282 -0.117 0.0269>
   rotation=<rot -6.2807 -90.2935 -0.6487>
  }]
 dyn_jacket_front_02_l [{
   shapeType=Box
   extentX=1.3643
   extentY=7.0538
   extentZ=9.656
   shapeName=dyn_jacket_front_02_l_box
   center=<3.0658 0.643 0.037>
   rotation=<rot -0.1882 -90.6696 0.1174>
  }]
 dyn_jacket_front_02_r [{
   shapeType=Box
   extentX=1.6643
   extentY=7.0538
   extentZ=9.656
   shapeName=dyn_jacket_front_02_l_box
   center=<3.0653 0.4949 0.0241>
   rotation=<rot -2.6162 -90.9423 0.1124>
  }]
 dyn_jacket_side_01_l [{
   shapeType=Box
   extentX=2.8044
   extentY=7.0958
   extentZ=11.2066
   shapeName=dyn_jacket_side_01_l_box
   center=<2.5283 -0.2167 0.0757>
   rotation=<rot 0.178 -90 -4.3177>
  }]
 dyn_jacket_side_01_r [{
   shapeType=Box
   extentX=2.8044
   extentY=7.0958
   extentZ=11.2066
   shapeName=dyn_jacket_side_01_r_box
   center=<2.5283 0.2167 -0.0757>
   rotation=<rot 0.178 -90 177.6625>
  }]
 dyn_jacket_side_02_l [{
   shapeType=Box
   extentX=1.5642
   extentY=5.0664
   extentZ=9.656
   shapeName=dyn_jacket_front_02_l_box
   center=<1.7664 0.2577 0.0217>
   rotation=<rot -0.1895 -90 0.922>
  }]
 dyn_jacket_side_02_r [{
   shapeType=Box
   extentX=2.2244
   extentY=5.0664
   extentZ=9.656
   shapeName=dyn_jacket_front_02_r_box
   center=<1.7664 0.1334 -0.0204>
   rotation=<rot -0.1895 -90 177.3415>
  }]
 dyn_pistolHolder [{
   shapeType=Capsule
   length=11.2151
   radius=7.2367
   shapeName=dyn_pistolHolder_capsule
   center=<-3.4884 3.0983 -0.3497>
   rotation=<rot -2.0801 -87.1036 -88.1708>
  }]
 pelvis [{
   shapeType=Capsule
   length=3.326
   radius=12.4758
   shapeName=pelvis_capsule
   center=<0.5034 -0.8882 0.0212>
   rotation=<rot 0.2164 -90.0001 0.4869>
  }]
 spine_01 [{
   shapeType=Capsule
   length=1.0839
   radius=12.2211
   shapeName=pelvis_capsule
   center=<0.5034 0.1379 0.0173>
   rotation=<rot 0.2164 -90.0001 0.4869>
  }]
 spine_02 [{
   shapeType=Capsule
   length=1.0839
   radius=12.2211
   shapeName=pelvis_capsule
   center=<0.5034 0.1379 0.0173>
   rotation=<rot 0.2164 -90.0001 0.4869>
  }]
 thigh_l [{shapeType=Capsule, length=23.1403, radius=7.799, shapeName=thigh_l_capsule, center=<-8.3903 0 1.4914>, rotation=<rot 90 0 0>}]
 thigh_r [{shapeType=Capsule, length=23.1403, radius=7.799, shapeName=thigh_l_capsule, center=<8.3903 0 -1.4914>, rotation=<rot 90 0 0>}]
}
constraints [
 {bone1Name=dyn_jacket_front_01_l, bone2Name=MSTR_jacket_front_l, swing1LimitDegrees=45, swing2LimitDegrees=20, twistLimitDegrees=15}
 {bone1Name=dyn_jacket_front_02_l, bone2Name=dyn_jacket_front_01_l, swing1LimitDegrees=30, swing2LimitDegrees=15, twistLimitDegrees=30}
 {bone1Name=dyn_jacket_side_01_l, bone2Name=MSTR_jacket_side_l, swing1LimitDegrees=45, swing2LimitDegrees=20, twistLimitDegrees=5}
 {bone1Name=dyn_jacket_side_02_l, bone2Name=dyn_jacket_side_01_l, swing1LimitDegrees=45, swing2LimitDegrees=20, twistLimitDegrees=5}
 {bone1Name=dyn_jacket_back_01_l, bone2Name=MSTR_jacket_back_l, swing1LimitDegrees=25, swing2LimitDegrees=15, twistLimitDegrees=5}
 {bone1Name=dyn_jacket_back_02_l, bone2Name=dyn_jacket_back_01_l, swing1LimitDegrees=25, swing2LimitDegrees=20, twistLimitDegrees=5}
 {
  bone1Name=dyn_jacket_back_01
  bone2Name=MSTR_jacket_back
  swing1LimitDegrees=25
  swing2LimitDegrees=5
  twistLimitDegrees=5
  twistMotion=Locked
 }
 {bone1Name=dyn_jacket_back_02, bone2Name=dyn_jacket_back_01, swing1LimitDegrees=25, swing2LimitDegrees=5, twistLimitDegrees=5}
 {bone1Name=dyn_jacket_back_01_r, bone2Name=MSTR_jacket_back_r, swing1LimitDegrees=25, swing2LimitDegrees=15, twistLimitDegrees=5}
 {bone1Name=dyn_jacket_back_02_r, bone2Name=dyn_jacket_back_01_r, swing1LimitDegrees=25, swing2LimitDegrees=20, twistLimitDegrees=5}
 {bone1Name=dyn_jacket_front_01_r, bone2Name=MSTR_jacket_front_r, swing1LimitDegrees=30, swing2LimitDegrees=20, twistLimitDegrees=15}
 {bone1Name=dyn_jacket_front_02_r, bone2Name=dyn_jacket_front_01_r, swing1LimitDegrees=30, swing2LimitDegrees=15, twistLimitDegrees=15}
 {bone1Name=dyn_jacket_side_01_r, bone2Name=MSTR_jacket_side_r, swing1LimitDegrees=45, swing2LimitDegrees=20, twistLimitDegrees=5}
 {bone1Name=dyn_jacket_side_02_r, bone2Name=dyn_jacket_side_01_r, swing1LimitDegrees=45, swing2LimitDegrees=20, twistLimitDegrees=5}
 {
  bone1Name=dyn_jacket_side_02_l
  bone2Name=dyn_jacket_front_02_l
  swing1LimitDegrees=45
  swing1Motion=Free
  swing2LimitDegrees=45
  swing2Motion=Free
  twistLimitDegrees=45
  twistMotion=Free
 }
 {
  bone1Name=dyn_jacket_back_02_l
  bone2Name=dyn_jacket_side_02_l
  swing1LimitDegrees=45
  swing1Motion=Free
  swing2LimitDegrees=45
  swing2Motion=Free
  twistLimitDegrees=45
  twistMotion=Free
 }
 {
  bone1Name=dyn_jacket_back_02
  bone2Name=dyn_jacket_back_02_l
  swing1LimitDegrees=45
  swing1Motion=Free
  swing2LimitDegrees=45
  swing2Motion=Locked
  twistLimitDegrees=5
 }
 {
  bone1Name=dyn_jacket_side_02_r
  bone2Name=dyn_jacket_front_02_r
  swing1LimitDegrees=45
  swing1Motion=Free
  swing2LimitDegrees=45
  swing2Motion=Free
  twistLimitDegrees=45
  twistMotion=Free
 }
 {
  bone1Name=dyn_jacket_back_02_r
  bone2Name=dyn_jacket_side_02_r
  swing1LimitDegrees=45
  swing1Motion=Free
  swing2LimitDegrees=45
  swing2Motion=Free
  twistLimitDegrees=45
  twistMotion=Free
 }
 {
  bone1Name=dyn_jacket_back_02
  bone2Name=dyn_jacket_back_02_r
  swing1LimitDegrees=45
  swing1Motion=Free
  swing2LimitDegrees=45
  swing2Motion=Locked
  twistLimitDegrees=5
 }
]
