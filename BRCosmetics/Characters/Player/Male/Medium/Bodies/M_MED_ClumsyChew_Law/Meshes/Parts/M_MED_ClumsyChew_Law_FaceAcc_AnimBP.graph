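# Face accessory AnimBP for the ClumsyChew_Law character's leaf, which follows the mouth and sways with physics.
# It copies the parent pose, constrains the leaf root to weighted lip, jaw and head bones, then runs spring-driven AnimDynamics on four leaf bones.
# A layered blend with DefaultSlot is weighted by a front-end idle curve.
# Computed gravity override, socket transforms and LeafMoveAlpha are not read by the AnimGraph.
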AnimBlueprint /BRCosmetics/Characters/Player/Male/Medium/Bodies/M_MED_ClumsyChew_Law/Meshes/Parts/M_MED_ClumsyChew_Law_FaceAcc_AnimBP
target_skeleton=&/M_MED_ClumsyChew_Law_FaceAcc_Skeleton
graphs [
 AnimGraph {nodes [
   Output Pose(Result=Layered blend per bone(
     BasePose=Use cached pose 'Start Pose'()
     BlendPoses_0=Slot 'DefaultSlot'()
     BlendWeights_0=Get Custom Idle is Playing Float() -> (Custom Idle is Playing Float:double)
    ) {layerSetup [{branchFilters [{blendDepth=-1, boneName=dyn_leaf_01}]}]}) {layerGroup=None}
   n1 = "Constraint - dyn_leaf_01 (PPPPPPP)"!(
    ComponentPose=Local To Component(LocalPose=Copy Pose From Mesh() {bUseAttachedParent=true})
    Alpha=1
    ConstraintWeights_0=0.2117
    ConstraintWeights_1:float=0.4003
    ConstraintWeights_2:float=0.0301
    ConstraintWeights_3:float=0.0183
    ConstraintWeights_4:float=0.0601
    ConstraintWeights_5:float=0.0013
    ConstraintWeights_6:float=0.2783
   ) -> (Pose) {
    alphaBoolBlend {blendOption=Linear}
    boneToModify {boneName=dyn_leaf_01}
    constraintSetup [
     {targetBone {boneName=R_lip_corner}}
     {targetBone {boneName=C_jaw}}
     {targetBone {boneName=head}}
     {targetBone {boneName=R_lip_upper_outer}}
     {targetBone {boneName=C_lip_lower_mid}}
     {targetBone {boneName=C_lip_upper_mid}}
     {targetBone {boneName=R_lip_lower_outer}}
    ]
    constraintWeights [1,1,1,1,1,1,1]
   }
   n2 = "Transform (Modify) Bone - Bone: dyn_leaf_01"(ComponentPose=@n1, Alpha=1, Translation="0,0,0.4", Rotation="0,0,0", Scale="1,1,1") {translationMode=BMM_Additive, alphaBoolBlend {blendOption=Linear}, boneToModify {boneName=dyn_leaf_01}, translation {z=0.4}}
   n3 = Get Simulation Space() -> (Simulation Space:AnimPhysSimSpaceType Enum)
   n4 = Anim Dynamics - Bone: dyn_leaf_02!(SimulationSpace:AnimPhysSimSpaceType Enum=@n3, ComponentPose=@n2, Alpha=1, GravityOverride:Vector="0,0,-980", ExternalForce="0,0,0") -> (Pose) {
    angularDampingOverride=0.996
    angularSpringConstant=500
    bAngularSpring=true
    bOverrideAngularDamping=true
    bUseGravityOverride=true
    simulationSpace=RootRelative
    alphaBoolBlend {blendOption=Linear}
    boundBone {boneName=dyn_leaf_02}
    componentAppliedLinearAccClamp=<3000 3000 3000>
    componentLinearAccScale=<0.5 0.5 0.5>
    componentLinearVelScale=<0.5 0.5 0.5>
    physicsBodyDefinitions [{
      boundBone {boneName=dyn_leaf_02}
      boxExtents=<7 7 7>
      constraintSetup {angularLimitsMax=<0 15 10>, angularLimitsMin=<0 -15 -20>, angularTarget=<1 0 0>}
      localJointOffset=<1 0 0>
     }]
   }
   n5 = Anim Dynamics - Bone: dyn_leaf_additive_01!(SimulationSpace:AnimPhysSimSpaceType Enum=@n3, ComponentPose=@n4, Alpha=1, GravityOverride:Vector="0,0,-980", ExternalForce="0,0,0") -> (Pose) {
    angularDampingOverride=0.994
    angularSpringConstant=800
    bAngularSpring=true
    bOverrideAngularDamping=true
    bUseGravityOverride=true
    simulationSpace=RootRelative
    alphaBoolBlend {blendOption=Linear}
    boundBone {boneName=dyn_leaf_additive_01}
    componentAppliedLinearAccClamp=<3000 3000 3000>
    componentLinearAccScale=<0.5 0.5 0.5>
    componentLinearVelScale=<0.5 0.5 0.5>
    physicsBodyDefinitions [{
      boundBone {boneName=dyn_leaf_additive_01}
      boxExtents=<9 9 9>
      constraintSetup {angularLimitsMax=<0 15 15>, angularLimitsMin=<0 -15 -15>, angularTarget=<1 0 0>}
      localJointOffset=<1.4 0.52 0>
     }]
   }
   n6 = Anim Dynamics - Bone: dyn_leaf_additive_02!(SimulationSpace:AnimPhysSimSpaceType Enum=@n3, ComponentPose=@n5, Alpha=1, GravityOverride:Vector="0,0,-980", ExternalForce="0,0,0") -> (Pose) {
    angularDampingOverride=0.995
    angularSpringConstant=800
    bAngularSpring=true
    bOverrideAngularDamping=true
    bUseGravityOverride=true
    simulationSpace=RootRelative
    alphaBoolBlend {blendOption=Linear}
    boundBone {boneName=dyn_leaf_additive_02}
    componentAppliedLinearAccClamp=<3000 3000 3000>
    componentLinearAccScale=<0.5 0.5 0.5>
    componentLinearVelScale=<0.5 0.5 0.5>
    physicsBodyDefinitions [{
      boundBone {boneName=dyn_leaf_additive_02}
      boxExtents=<10 10 10>
      constraintSetup {angularLimitsMax=<0 15 15>, angularLimitsMin=<0 -15 -15>, angularTarget=<1 0 0>}
      localJointOffset=<1.5 0.5 0>
     }]
   }
   n7 = Anim Dynamics - Bone: dyn_leaf_additive_03!(SimulationSpace:AnimPhysSimSpaceType Enum=@n3, ComponentPose=@n6, Alpha=1, GravityOverride:Vector="0,0,-980", ExternalForce="0,0,0") -> (Pose) {
    angularDampingOverride=0.992
    angularSpringConstant=800
    bAngularSpring=true
    bOverrideAngularDamping=true
    bUseGravityOverride=true
    simulationSpace=RootRelative
    alphaBoolBlend {blendOption=Linear}
    boundBone {boneName=dyn_leaf_additive_03}
    componentAppliedLinearAccClamp=<3000 3000 3000>
    componentLinearAccScale=<0.5 0.5 0.5>
    componentLinearVelScale=<0.5 0.5 0.5>
    physicsBodyDefinitions [{
      boundBone {boneName=dyn_leaf_additive_03}
      boxExtents=<9 9 9>
      constraintSetup {angularLimitsMax=<0 15 15>, angularLimitsMin=<0 -15 -15>, angularTarget=<1 0 0>}
      localJointOffset=<1.3 0.54 0>
     }]
   }
   Save cached pose 'Start Pose'(Pose=Component To Local(ComponentPose=@n7)) {cacheName=Start Pose}
  ]}
 EventGraph {nodes [
   Event BlueprintUpdateAnimation()
   > n1 = Gravity Override State Function DEPRECATED(
    CharacterPartAnimInstance=Self-Reference()
    GameTimeInSeconds=GetGameTimeInSeconds()
    GravityOverrideParams=Get Gravity Override Params()
    RBAN Params=Get RBAN Params()
   )
   > Set Gravity Override(Gravity Override=@n1)
   n2 = Get HeadSkel() -> (HeadSkel:Skeletal Mesh Component Budgeted Object Reference)
   > Set R_lip_corner(R_lip_corner:Transform=GetSocketTransform(self=@n2, InSocketName=R_lip_corner, TransformSpace=RTS_World)) -> (Output_Get:Transform)
   > Set C_jaw(C_jaw:Transform=GetSocketTransform(self=@n2, InSocketName=C_jaw, TransformSpace=RTS_ParentBoneSpace)) -> (Output_Get:Transform)
   > Set R_lip_lower_outer(R_lip_lower_outer:Transform=GetSocketTransform(self=@n2, InSocketName=R_lip_lower_outer, TransformSpace=RTS_World)) -> (Output_Get:Transform)
   > Set C_lip_lower_mid(C_lip_lower_mid:Transform=GetSocketTransform(self=@n2, InSocketName=C_lip_lower_mid, TransformSpace=RTS_World)) -> (Output_Get:Transform)
   > Set LeafMoveAlpha(LeafMoveAlpha:double=Remap Float(
     Input=Get!() -> (C_jaw_Location:Vector, C_jaw_Rotation_Roll:float, C_jaw_Rotation_Pitch:float, C_jaw_Rotation_Yaw:float, C_jaw_Scale:Vector).C_jaw_Rotation_Roll
     OldRange=(X=10,Y=30)
     NewRange=(X=0,Y=1)
     Clamp=true
    ) -> (Output=~)) -> (Output_Get:double)
   > Branch(Condition=Get bIsInFrontEnd())
   > Set Custom Idle is Playing Float(Custom Idle is Playing Float:double=GetCurveValue(self=GetOwnerFrontEndAnimInstance(), CurveName=Frontend_Idle_Bypass)) -> (Output_Get:double)
   n3 = Event OnCharacterCustomizationLoaded()
   Set HeadMesh(
    execute=[@n3.then, @n4.then]
    HeadMesh:Custom Character Part Object Reference=Get Array Item with EFort Custom Part Type(Array:Array of Custom Character Part Object References=@n3.CustomParts, PartType=Head) -> (Output:Custom Character Part Object Reference)
   ) -> (Output_Get:Custom Character Part Object Reference)
   > Set HeadSkel(HeadSkel:Skeletal Mesh Component Budgeted Object Reference=Get Array Item with EFort Custom Part Type(Array:Array of Skeletal Mesh Component Budgeted Object References=@n3.CustomMeshParts, PartType=Head) -> (Output:Skeletal Mesh Component Budgeted Object Reference)) -> (Output_Get:Skeletal Mesh Component Budgeted Object Reference)
   > Set Simulation Space(Simulation Space:AnimPhysSimSpaceType Enum=Select(Option 0:AnimPhysSimSpaceType Enum, Option 1:AnimPhysSimSpaceType Enum=World, Index=Get bIsInFrontEnd()) -> (ReturnValue:AnimPhysSimSpaceType Enum)) -> (Output_Get:AnimPhysSimSpaceType Enum)
   n4 = Event BlueprintInitializeAnimation()
  ]}
]
variables [
 C_jaw {type=Transform, default {location=<0 0 0>, rotation=<rot 0 0 0>, scale=<1 1 1>}}
 C_lip_lower_mid {type=Transform, default {location=<0 0 0>, rotation=<rot 0 0 0>, scale=<1 1 1>}}
 Custom Idle is Playing Float {type=number}
 Gravity Override {type=Vector}
 Gravity Override Params {type=GravityOverrideParamsStruct}
 HeadMesh {type=CustomCharacterPart, default=null}
 HeadSkel {type=SkeletalMeshComponentBudgeted}
 LeafMoveAlpha {type=number}
 R_lip_corner {type=Transform, default {location=<0 0 0>, rotation=<rot 0 0 0>, scale=<1 1 1>}}
 R_lip_lower_outer {type=Transform, default {location=<0 0 0>, rotation=<rot 0 0 0>, scale=<1 1 1>}}
 RBAN Params {
  type=RBANParamsStruct
  default {
   dBNO {simSpaceSettings {
     maxAngularAcceleration=1000
     maxAngularVelocity=3
     maxLinearAcceleration=300
     maxLinearVelocity=700
     velocityScaleZ=0
     externalLinearDragV=<0 0.5 0>
    }}
   dBNO_Carried {simSpaceSettings {
     maxAngularAcceleration=1000
     maxAngularVelocity=3
     maxLinearAcceleration=300
     maxLinearVelocity=700
     velocityScaleZ=0
     externalLinearDragV=<0 0.5 0>
    }}
   emoteOrMelee {simSpaceSettings {
     maxAngularAcceleration=1000
     maxAngularVelocity=3
     maxLinearAcceleration=300
     maxLinearVelocity=700
     velocityScaleZ=0
     externalLinearDragV=<0 0.5 0>
    }}
   front_End {simSpaceSettings {
     maxAngularAcceleration=1000
     maxAngularVelocity=3
     maxLinearAcceleration=300
     maxLinearVelocity=700
     velocityScaleZ=0
     externalLinearDragV=<0 0.5 0>
    }}
   onGround_Crouch_Moving {simSpaceSettings {
     maxAngularAcceleration=1000
     maxAngularVelocity=3
     maxLinearAcceleration=300
     maxLinearVelocity=700
     velocityScaleZ=0
     externalLinearDragV=<0 0.5 0>
    }}
   onGround_Falling {simSpaceSettings {
     maxAngularAcceleration=1000
     maxAngularVelocity=3
     maxLinearAcceleration=300
     maxLinearVelocity=700
     velocityScaleZ=0
     externalLinearDragV=<0 0.5 0>
    }}
   onGround_Moving {simSpaceSettings {
     maxAngularAcceleration=1000
     maxAngularVelocity=3
     maxLinearAcceleration=300
     maxLinearVelocity=700
     velocityScaleZ=0
     externalLinearDragV=<0 0.5 0>
    }}
   onGround_Standing {simSpaceSettings {
     maxAngularAcceleration=1000
     maxAngularVelocity=3
     maxLinearAcceleration=300
     maxLinearVelocity=700
     velocityScaleZ=0
     externalLinearDragV=<0 0.5 0>
    }}
   rail_Grinding {simSpaceSettings {
     maxAngularAcceleration=1000
     maxAngularVelocity=3
     maxLinearAcceleration=300
     maxLinearVelocity=700
     velocityScaleZ=0
     worldAlpha=0.1
     externalLinearDragV=<0 0.5 0>
    }}
   riding_Moving {simSpaceSettings {
     maxAngularAcceleration=1000
     maxAngularVelocity=3
     maxLinearAcceleration=300
     maxLinearVelocity=700
     velocityScaleZ=0
     externalLinearDragV=<0 0.5 0>
    }}
   riding_Moving_Targeting {simSpaceSettings {
     maxAngularAcceleration=1000
     maxAngularVelocity=3
     maxLinearAcceleration=300
     maxLinearVelocity=700
     velocityScaleZ=0
     externalLinearDragV=<0 0.5 0>
    }}
   riding_Standing {simSpaceSettings {
     maxAngularAcceleration=1000
     maxAngularVelocity=3
     maxLinearAcceleration=300
     maxLinearVelocity=700
     velocityScaleZ=0
     externalLinearDragV=<0 0.5 0>
    }}
   skydive_DivingDown {simSpaceSettings {
     maxAngularAcceleration=1000
     maxAngularVelocity=3
     maxLinearAcceleration=300
     maxLinearVelocity=700
     velocityScaleZ=0
     externalLinearDragV=<0 0.5 0>
    }}
   skydive_FloatingDown {
    componentLinearAccScale=<0 0.5 0>
    componentLinearVelScale=<0 0.5 0>
    simSpaceSettings {
     maxAngularAcceleration=1000
     maxAngularVelocity=3
     maxLinearAcceleration=300
     maxLinearVelocity=700
     velocityScaleZ=0
     externalLinearDragV=<0 0.5 0>
    }
   }
   skydive_GoingUp {simSpaceSettings {
     maxAngularAcceleration=1000
     maxAngularVelocity=3
     maxLinearAcceleration=300
     maxLinearVelocity=700
     velocityScaleZ=0
     externalLinearDragV=<0 0.5 0>
    }}
   skydive_Parachuting {
    componentLinearAccScale=<0 0.1 0.3>
    componentLinearVelScale=<0 0.1 0.3>
    simSpaceSettings {
     maxAngularAcceleration=1000
     maxAngularVelocity=3
     maxLinearAcceleration=300
     maxLinearVelocity=700
     velocityScaleZ=0
     externalLinearDragV=<0 0.5 0>
    }
   }
   skydive_Surfing {simSpaceSettings {
     maxAngularAcceleration=1000
     maxAngularVelocity=3
     maxLinearAcceleration=300
     maxLinearVelocity=700
     velocityScaleZ=0
     externalLinearDragV=<0 0.5 0>
    }}
   swimming {simSpaceSettings {
     maxAngularAcceleration=1000
     maxAngularVelocity=3
     maxLinearAcceleration=300
     maxLinearVelocity=700
     velocityScaleZ=0
     externalLinearDragV=<0 0.5 0>
    }}
   tactical_Sprint {simSpaceSettings {
     maxAngularAcceleration=1000
     maxAngularVelocity=3
     maxLinearAcceleration=300
     maxLinearVelocity=700
     velocityScaleZ=0
     externalLinearDragV=<0 0.5 0>
    }}
   ziplining {simSpaceSettings {
     maxAngularAcceleration=1000
     maxAngularVelocity=3
     maxLinearAcceleration=300
     maxLinearVelocity=700
     velocityScaleZ=0
     externalLinearDragV=<0 0.5 0>
    }}
  }
 }
 Simulation Space {type=AnimPhysSimSpaceType, default=RootRelative}
]
